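# Parent material for the PeacefulPoem character cape, layering an animated, tinted noise glow onto the shared 2023 CharacterParent PreFX/PostFX functions.
# A custom UV texture's R and G channels shape the mask and blend tiled mesh UVs with remapped coordinates.
# These feed a custom panner function whose noise, multiplied by the mask and EmissiveColor, is blended into emissive.
# The texture's R channel also drives base color darkening, with both results published through named reroutes.
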Material /BRCosmetics/Characters/Player/Male/Medium/Bodies/M_MED_PeacefulPoem/Materials/M_PeacefulPoem_Cape_Parent1
graph {
 properties {
  bUseMaterialAttributes=true
  bUsedWithClothing=true
  bUsedWithMeshDeformer=true
  bUsedWithMorphTargets=true
  bUsedWithNanite=true
  bUsedWithSkeletalMesh=true
  pixelDepthOffsetMode=PDOM_Legacy
  previewMesh=/BRCosmetics/Characters/Player/Male/Large/Bodies/M_LRG_PeacefulPoem_Pumped/Meshes/Demo/PeacefulPoem_Pumped_Cape_LOD0
  shadingModel=MSM_FromMaterialExpression
  usedShadingModels=Default Lit | Subsurface | Clear Coat | Subsurface Profile | Cloth
 }
 nodes [
  n1 = MaterialFunctionCall() {materialFunction=&MF_CharacterParent_2023_PreFX}
  MaterialFunctionCall(Material=@n1.Result, M=@n1.Result, Pre Skinned Local Position=@n1.Result, Distance Blend=@n1.Result, Fresnel=@n1.Result) {materialFunction=&MF_CharacterParent_2023_PostFX}
  n2 = TextureSampleParameter2D() {
   parameterName=CustomUVTexture
   texture=/BRCosmetics/Characters/Player/Male/Large/Bodies/M_LRG_PeacefulPoem_Pumped/Materials/T_PeacefulPoemPumped_UVs_Universal_D
  }
  n3 = OneMinus(None=@n2.G)
  n4 = MaterialFunctionCall(Color=NamedRerouteUsage()) {materialFunction=&SplitComponents}
  Multiply(A=@n3, B=@n4.B)
  n5 = NamedRerouteUsage()
  NamedRerouteDeclaration(None=LinearInterpolate(A=@n5, B=Multiply(A=@n5) {constB=0.1}, Alpha=Power(Base=@n2.R))) {name=outBaseColor, nodeColor=<rgba 1 0.8203 0 1>}
  n6 = LinearInterpolate(A=@n4.B, B=Multiply(A=@n2.R) {constB=0.8}, Alpha=@n2.G)
  n7 = TextureCoordinate()
  n8 = Constant2Vector() {g=1, r=2}
  n9 = VectorParameter() {parameterName=NoiseTiling, defaultValue=<rgba 3 0.3 0 1>}
  n10 = Multiply(A=ComponentMask(None=TextureCoordinate()) {g=true}, B=@n9.R) {constB=3}
  n11 = Multiply(A=@n2.R, B=@n9.G) {constB=0.3}
  n12 = MaterialFunctionCall!(
   UV=LinearInterpolate(A=Multiply(A=@n7, B=@n8), B=AppendVector(A=@n10, B=@n11), Alpha=@n2.G)
   Speed=ComponentMask(None=VectorParameter() {parameterName=Speed, defaultValue=<rgba 0 -0.025 0 1>}.RGB) {g=true, r=true}
   Mask=@n6
  ) {materialFunction=/BRCosmetics/Characters/Player/Male/Large/Bodies/M_LRG_PeacefulPoem_Pumped/Materials/MF_PeaceulPoemPumpedPanner}
  n13 = Multiply(A=@n6, B=LinearInterpolate(A=@n12.RawNoise, B=@n12.RawNoise, Alpha=@n2.G))
  n14 = VectorParameter() {parameterName=EmissiveColor, defaultValue=<rgba 0.2623 0.0033 0.7758 1>}
  n15 = Add(A=Ceil(None=@n2.R), B=@n3)
  NamedRerouteDeclaration(None=LinearInterpolate(A=NamedRerouteUsage(), B=Multiply(A=@n13, B=@n14.RGB), Alpha=Saturate(None=@n15))) {name=outEmissive, nodeColor=<rgba 1 0 0.5625 1>}
 ]
}
