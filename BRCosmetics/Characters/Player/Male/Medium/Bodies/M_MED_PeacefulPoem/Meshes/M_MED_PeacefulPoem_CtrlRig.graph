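# Corrective deformation Control Rig for the M_MED_PeacefulPoem male player body, run after the shared DeformRig function.
# Spherical pose readers on clavicles and upper arms drive shoulder pad offsets, with clavicle-raise values blending combined poses and an extra parachute-gated adjustment.
# Wrist anim attributes shift built-in pickaxe hand roots.
# Thigh readers drive hip armor flaps, whose rear base is parent-constrained across pelvis and spine; the head-rotation nodes are left unconnected.
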ControlRigBlueprint /BRCosmetics/Characters/Player/Male/Medium/Bodies/M_MED_PeacefulPoem/Meshes/M_MED_PeacefulPoem_CtrlRig
graphs [
 New Function_ContainedGraph {nodes [
   n1 = RigVMFunctionEntryNode()
   RigVMFunction_Sequence = &RigVMFunction_Sequence(ExecuteContext=@n1.ExecuteContext)
   RigVMFunction_Sequence_1 = RigVMAggregateNode(ExecuteContext=@RigVMFunction_Sequence.A)
   RigVMFunction_MathBoolToFloat = &RigVMFunction_MathBoolToFloat(Value=RigVMVariableNode() {variable=IsParachuteOpen}.Value)
   Multiply = &RigVMFunction_MathDoubleMul(
    B=RigVMDispatchNode(BoneName=root, Default=0, Name=deformRig_l_upperarm_up_alpha) {template="DISPATCH_RigDispatch_GetAnimAttribute(in Name,in BoneName,in Default,out Value,out Found)"}.Value
    A=@RigVMFunction_MathBoolToFloat.Result
   )
   &RigUnit_ModifyTransforms(
    ItemToModify=((Item=(Type=Bone,Name="deform_shoulderpad_01_l"),Transform=(Rotation=(X=0,Y=0,Z=0,W=1),Translation=(X=2,Y=0,Z=1),Scale3D=(X=0.9,Y=0.9,Z=0.9))))
    WeightMinimum=0
    Weight=@Multiply.Result
    ExecutePin=@RigVMFunction_Sequence_1.D
   )
   > &RigUnit_ModifyTransforms(
    ItemToModify=((Item=(Type=Bone,Name="deform_shoulderpad_01_r"),Transform=(Rotation=(X=0,Y=0,Z=0,W=1),Translation=(X=-2,Y=0,Z=-1.5),Scale3D=(X=0.9,Y=0.9,Z=0.9))))
    WeightMinimum=0
    Weight=&RigVMFunction_MathDoubleMul(
     B=RigVMDispatchNode(BoneName=root, Default=0, Name=deformRig_r_upperarm_up_alpha) {template="DISPATCH_RigDispatch_GetAnimAttribute(in Name,in BoneName,in Default,out Value,out Found)"}.Value
     A=@RigVMFunction_MathBoolToFloat.Result
    ).Result
   )
   n2 = &RigUnit_SphericalPoseReader(
    ActiveRegionScaleFactors=(PositiveWidth=2,NegativeWidth=1,PositiveHeight=1,NegativeHeight=2)
    ActiveRegionSize=0.12
    Debug=(bDrawDebug=true,DebugScale=25,DebugSegments=20,DebugThickness=0.25)
    DriverAxis=(X=1,Y=0,Z=0)
    DriverItem=(Type=Bone,Name="upperarm_l")
    FalloffRegionScaleFactors=(PositiveWidth=1,NegativeWidth=1,PositiveHeight=1,NegativeHeight=1)
    FalloffSize=0.6
    OptionalParentItem=(Type=Bone,Name="clavicle_l")
    RotationOffset=(X=60.3761,Y=0,Z=33.2704)
    ExecutePin=@RigVMFunction_Sequence_1.C
   )
   > n3 = &RigUnit_SphericalPoseReader(
    ActiveRegionScaleFactors=(PositiveWidth=1,NegativeWidth=2,PositiveHeight=1,NegativeHeight=2)
    ActiveRegionSize=0.12
    Debug=(bDrawDebug=true,DebugScale=25,DebugSegments=20,DebugThickness=0.25)
    DriverAxis=(X=-1,Y=0,Z=0)
    DriverItem=(Type=Bone,Name="upperarm_r")
    FalloffRegionScaleFactors=(PositiveWidth=1,NegativeWidth=1,PositiveHeight=1,NegativeHeight=1)
    FalloffSize=0.6
    FlipHeightScaling=true
    FlipWidthScaling=true
    OptionalParentItem=(Type=Bone,Name="clavicle_r")
    RotationOffset=(X=121.4949,Y=-8.2258,Z=-147.8671)
   )
   > &RigUnit_ModifyTransforms(
    ItemToModify="((Item=(Type=Bone,Name=\"deform_shoulderpad_01_l\"),Transform=(Rotation=(X=0.1059,Y=-0.1071,Z=0.2093,W=0.9662),Translation=(X=-0.3355,Y=-3.1637,Z=-0.6534),Scale3D=(X=1,Y=1,Z=1))),(Item=(Type=Bone,Name=\"deform_shoulderpad_02_l\"),Transform=(Rotation=(X=0.038,Y ...(truncated)"
    WeightMinimum=0
    Weight=@n2.OutputParam
   )
   > Get Mirror Transform_3 = RigVMFunctionReferenceNode(transform=(Rotation=(X=0.1059,Y=-0.1071,Z=0.2093,W=0.9662),Translation=(X=-0.3355,Y=-3.1637,Z=-0.6534),Scale3D=(X=1,Y=1,Z=1)))
   > &RigUnit_ModifyTransforms(
    ItemToModify="((Item=(Type=Bone,Name=\"deform_shoulderpad_01_r\"),Transform=(Rotation=(X=0,Y=-0.1632,Z=0,W=0.9866),Translation=(X=-2.3381,Y=4.0995,Z=-3.9185),Scale3D=(X=1,Y=1,Z=1))),(Item=(Type=Bone,Name=\"deform_shoulderpad_02_r\"),Transform=(Rotation=(X=0.038,Y=-0.0285,Z= ...(truncated)"
    WeightMinimum=0
    ItemToModify.0.Transform=@"Get Mirror Transform_3".mirrorTransform
    Weight=@n3.OutputParam
   )
   n4 = RigVMVariableNode() {variable=l_clav_raise_spr}
   > &RigUnit_ModifyTransforms(
    ItemToModify="((Item=(Type=Bone,Name=\"deform_shoulderpad_01_l\"),Transform=(Rotation=(X=0.2994,Y=0.1837,Z=0.0109,W=0.9362),Translation=(X=-0.6832,Y=-0.0061,Z=-2.6009),Scale3D=(X=1,Y=1,Z=1))),(Item=(Type=Bone,Name=\"deform_shoulderpad_02_l\"),Transform=(Rotation=(X=0,Y=0,Z= ...(truncated)"
    WeightMinimum=0
    Weight=&RigVMFunction_MathDoubleRemap(
     SourceMaximum=1.65
     SourceMinimum=1
     TargetMaximum=1
     TargetMinimum=0
     Value=&RigVMFunction_MathDoubleAdd(A=@n2.OutputParam, B=@n4.Value).Result
    ).Result
   )
   > Get Mirror Transform_1_1_1 = RigVMFunctionReferenceNode(transform=(Rotation=(X=0.2994,Y=0.1837,Z=0.0109,W=0.9362),Translation=(X=-0.6832,Y=-0.0061,Z=-2.6009),Scale3D=(X=1,Y=1,Z=1)))
   n5 = RigVMVariableNode() {variable=r_clav_raise_spr}
   > &RigUnit_ModifyTransforms(
    ItemToModify="((Item=(Type=Bone,Name=\"deform_shoulderpad_01_r\"),Transform=(Rotation=(X=0,Y=-0.1962,Z=0,W=0.9806),Translation=(X=0.4714,Y=0,Z=2.3308),Scale3D=(X=0.9,Y=0.9,Z=0.9))),(Item=(Type=Bone,Name=\"deform_shoulderpad_02_r\"),Transform=(Rotation=(X=0,Y=0,Z=0,W=1),Tran ...(truncated)"
    WeightMinimum=0
    ItemToModify.0.Transform=@"Get Mirror Transform_1_1_1".mirrorTransform
    Weight=&RigVMFunction_MathDoubleRemap(
     SourceMaximum=1.65
     SourceMinimum=1
     TargetMaximum=1
     TargetMinimum=0
     Value=&RigVMFunction_MathDoubleAdd(A=@n3.OutputParam, B=@n5.Value).Result
    ).Result
   )
   n6 = &RigUnit_SphericalPoseReader(
    ActiveRegionScaleFactors=(PositiveWidth=2.5,NegativeWidth=3,PositiveHeight=2,NegativeHeight=3)
    ActiveRegionSize=0.1
    Debug=(bDrawDebug=false,DebugScale=25,DebugSegments=20,DebugThickness=0.25)
    DriverAxis=(X=1,Y=0,Z=0)
    DriverItem=(Type=Bone,Name="upperarm_l")
    FalloffRegionScaleFactors=(PositiveWidth=1,NegativeWidth=1,PositiveHeight=1,NegativeHeight=1)
    FalloffSize=0.6
    OptionalParentItem=(Type=Bone,Name="clavicle_l")
    RotationOffset=(X=0,Y=20,Z=0)
    ExecutePin=@RigVMFunction_Sequence_1.B
   )
   > n7 = &RigUnit_SphericalPoseReader(
    ActiveRegionScaleFactors=(PositiveWidth=2.5,NegativeWidth=3,PositiveHeight=2,NegativeHeight=3)
    ActiveRegionSize=0.1
    Debug=(bDrawDebug=false,DebugScale=25,DebugSegments=20,DebugThickness=0.25)
    DriverAxis=(X=-1,Y=0,Z=0)
    DriverItem=(Type=Bone,Name="upperarm_r")
    FalloffRegionScaleFactors=(PositiveWidth=1,NegativeWidth=1,PositiveHeight=1,NegativeHeight=1)
    FalloffSize=0.6
    FlipHeightScaling=true
    OptionalParentItem=(Type=Bone,Name="clavicle_r")
    RotationOffset=(X=180,Y=-20,Z=180)
   )
   > &RigUnit_ModifyTransforms(
    ItemToModify="((Item=(Type=Bone,Name=\"deform_shoulderpad_01_l\"),Transform=(Rotation=(X=0,Y=-0.3838,Z=0,W=0.9234),Translation=(X=-1.106,Y=0.0205,Z=2.4),Scale3D=(X=0.9,Y=0.9,Z=0.9))),(Item=(Type=Bone,Name=\"deform_shoulderpad_02_l\"),Transform=(Rotation=(X=0,Y=-0.0499,Z=0,W ...(truncated)"
    WeightMinimum=0
    Weight=@n6.OutputParam
   )
   > Get Mirror Transform = RigVMFunctionReferenceNode(transform=(Rotation=(X=0,Y=-0.3838,Z=0,W=0.9234),Translation=(X=-1.106,Y=0,Z=1),Scale3D=(X=0.9,Y=0.9,Z=0.9)))
   > Get Mirror Transform_1 = RigVMFunctionReferenceNode(transform=(Rotation=(X=0,Y=-0.0499,Z=0,W=0.9988),Translation=(X=-1.9805,Y=-0.0338,Z=-0.5915),Scale3D=(X=1,Y=1,Z=1)))
   > Get Mirror Transform_2 = RigVMFunctionReferenceNode(transform=(Rotation=(X=0,Y=-0.0223,Z=0,W=0.9998),Translation=(X=-0.7085,Y=-0.0118,Z=0.0199),Scale3D=(X=1,Y=1,Z=1)))
   > &RigUnit_ModifyTransforms(
    ItemToModify="((Item=(Type=Bone,Name=\"deform_shoulderpad_01_r\"),Transform=(Rotation=(X=0,Y=-0.1962,Z=0,W=0.9806),Translation=(X=0.4714,Y=0,Z=2.3308),Scale3D=(X=0.9,Y=0.9,Z=0.9))),(Item=(Type=Bone,Name=\"deform_shoulderpad_02_r\"),Transform=(Rotation=(X=0,Y=-0.0499,Z=0,W=0 ...(truncated)"
    WeightMinimum=0
    ItemToModify.0.Transform=@"Get Mirror Transform".mirrorTransform
    ItemToModify.1.Transform=@"Get Mirror Transform_1".mirrorTransform
    ItemToModify.2.Transform=@"Get Mirror Transform_2".mirrorTransform
    Weight=@n7.OutputParam
   )
   Add = &RigVMFunction_MathDoubleAdd(A=@n6.OutputParam, B=RigVMVariableNode() {variable=l_clav_raise_spr}.Value)
   Remap = &RigVMFunction_MathDoubleRemap(SourceMaximum=1.5, SourceMinimum=1, TargetMaximum=1, TargetMinimum=0, Value=@Add.Result)
   > &RigUnit_ModifyTransforms(
    ItemToModify="((Item=(Type=Bone,Name=\"deform_shoulderpad_01_l\"),Transform=(Rotation=(X=0,Y=0,Z=0,W=1),Translation=(X=0.8057,Y=1.0882,Z=3.8572),Scale3D=(X=1,Y=1,Z=1))),(Item=(Type=Bone,Name=\"deform_shoulderpad_02_l\"),Transform=(Rotation=(X=0,Y=0,Z=0,W=1),Translation=(X=0 ...(truncated)"
    WeightMinimum=0
    Weight=@Remap.Result
   )
   > Get Mirror Transform_1_1 = RigVMFunctionReferenceNode(transform=(Rotation=(X=0,Y=0,Z=0,W=1),Translation=(X=0.8057,Y=1.0882,Z=3.8572),Scale3D=(X=1,Y=1,Z=1)))
   n8 = RigVMVariableNode() {variable=r_clav_raise_spr}
   > &RigUnit_ModifyTransforms(
    ItemToModify="((Item=(Type=Bone,Name=\"deform_shoulderpad_01_r\"),Transform=(Rotation=(X=0,Y=-0.1962,Z=0,W=0.9806),Translation=(X=0.4714,Y=0,Z=2.3308),Scale3D=(X=0.9,Y=0.9,Z=0.9))),(Item=(Type=Bone,Name=\"deform_shoulderpad_02_r\"),Transform=(Rotation=(X=0,Y=0,Z=0,W=1),Tran ...(truncated)"
    WeightMinimum=0
    ItemToModify.0.Transform=@"Get Mirror Transform_1_1".mirrorTransform
    Weight=&RigVMFunction_MathDoubleRemap(
     SourceMaximum=1.6
     SourceMinimum=1
     TargetMaximum=1
     TargetMinimum=0
     Value=&RigVMFunction_MathDoubleAdd(A=@n7.OutputParam, B=@n8.Value).Result
    ).Result
   )
   n9 = &RigUnit_SphericalPoseReader(
    ActiveRegionScaleFactors=(PositiveWidth=1,NegativeWidth=1,PositiveHeight=1,NegativeHeight=1)
    ActiveRegionSize=0.05
    Debug=(bDrawDebug=false,bDraw2D=False,bDrawLocalAxes=False,DebugScale=25,DebugSegments=20,DebugThickness=0.25)
    DriverAxis=(X=1,Y=0,Z=0)
    DriverItem=(Type=Bone,Name="clavicle_l")
    FalloffRegionScaleFactors=(PositiveWidth=1,NegativeWidth=1,PositiveHeight=1,NegativeHeight=1)
    FalloffSize=0.3
    FlipHeightScaling=False
    FlipWidthScaling=False
    OptionalParentItem=(Type=Bone,Name="None")
    RotationOffset=(X=0,Y=-25,Z=0)
    ExecutePin=@RigVMFunction_Sequence_1.A
   )
   > n10 = &RigUnit_SphericalPoseReader(
    ActiveRegionScaleFactors=(PositiveWidth=1,NegativeWidth=1,PositiveHeight=1,NegativeHeight=1)
    ActiveRegionSize=0.05
    Debug=(bDrawDebug=false,bDraw2D=False,bDrawLocalAxes=False,DebugScale=25,DebugSegments=20,DebugThickness=0.25)
    DriverAxis=(X=-1,Y=0,Z=0)
    DriverItem=(Type=Bone,Name="clavicle_r")
    FalloffRegionScaleFactors=(PositiveWidth=1,NegativeWidth=1,PositiveHeight=1,NegativeHeight=1)
    FalloffSize=0.3
    FlipHeightScaling=False
    FlipWidthScaling=False
    OptionalParentItem=(Type=Bone,Name="None")
    RotationOffset=(X=180,Y=-25,Z=0)
   )
   > &RigUnit_ModifyTransforms(
    ItemToModify=((Item=(Type=Bone,Name="deform_shoulderpad_01_l"),Transform=(Rotation=(X=0,Y=-0.0184,Z=0,W=0.9998),Translation=(X=8.8374,Y=1.4018,Z=-1.2706),Scale3D=(X=1,Y=1.1,Z=1))))
    WeightMinimum=0
    Weight=@n9.OutputParam
   )
   > Get Mirror Transform_5 = RigVMFunctionReferenceNode(transform=(Rotation=(X=0,Y=-0.0184,Z=0,W=0.9998),Translation=(X=8.8374,Y=1.4018,Z=-1.2706),Scale3D=(X=1,Y=1.1,Z=1)))
   > &RigUnit_ModifyTransforms(
    ItemToModify=((Item=(Type=Bone,Name="deform_shoulderpad_01_r"),Transform=(Rotation=(X=0.0106,Y=-0.1216,Z=0.0865,W=0.9887),Translation=(X=-2.9701,Y=0,Z=-2.6751),Scale3D=(X=1,Y=1,Z=1))))
    WeightMinimum=0
    ItemToModify.0.Transform=@"Get Mirror Transform_5".mirrorTransform
    Weight=@n10.OutputParam
   )
   > RigVMVariableNode(Value=@n9.OutputParam) {variable=l_clav_raise_spr}
   > RigVMVariableNode(Value=@n10.OutputParam) {variable=r_clav_raise_spr}
   RigVMFunctionReturnNode(ExecuteContext=@RigVMFunction_Sequence.B)
  ]}
 New Function_ContainedGraph {nodes [
   RigVMFunctionEntryNode()
   > &RigUnit_ModifyTransforms(
    ItemToModify=((Item=(Type=Bone,Name="deform_pickaxe_hand_root_l"),Transform=(Rotation=(X=0,Y=0,Z=0,W=1),Translation=(X=2.5722,Y=0,Z=0),Scale3D=(X=1,Y=1,Z=1))))
    WeightMinimum=0
    Weight=RigVMDispatchNode(BoneName=root, Default=0, Name=deformRig_l_wrist_up_alpha) {template="DISPATCH_RigDispatch_GetAnimAttribute(in Name,in BoneName,in Default,out Value,out Found)"}.Value
   )
   > &RigUnit_ModifyTransforms(
    ItemToModify=((Item=(Type=Bone,Name="deform_pickaxe_hand_root_r"),Transform=(Rotation=(X=0,Y=0,Z=0,W=1),Translation=(X=-2.5722,Y=0,Z=0),Scale3D=(X=1,Y=1,Z=1))))
    WeightMinimum=0
    Weight=RigVMDispatchNode(BoneName=root, Default=0, Name=deformRig_r_wrist_up_alpha) {template="DISPATCH_RigDispatch_GetAnimAttribute(in Name,in BoneName,in Default,out Value,out Found)"}.Value
   )
   > RigVMFunctionReturnNode()
  ]}
 New Function_ContainedGraph {nodes [
   RigVMFunctionEntryNode()
   > &RigUnit_ParentConstraint(
    AdvancedSettings=(RotationOrderForFilter=XZY)
    Child=(Type=Bone,Name="dyn_rear_flap_01_base")
    Parents=((Item=(Type=Bone,Name="spine_03"),Weight=0.0453),(Item=(Type=Bone,Name="spine_02"),Weight=0.0647),(Item=(Type=Bone,Name="spine_01"),Weight=0.0782),(Item=(Type=Bone,Name="pelvis"),Weight=0.8118))
    Weight=1
   )
   > n1 = &RigUnit_SphericalPoseReader(
    ActiveRegionScaleFactors=(PositiveWidth=3.5,NegativeWidth=0.5,PositiveHeight=3,NegativeHeight=2)
    ActiveRegionSize=0.1
    Debug=(bDrawDebug=true,DebugScale=25,DebugSegments=20,DebugThickness=0.25)
    DriverAxis=(X=-1,Y=0,Z=0)
    DriverItem=(Type=Bone,Name="thigh_l")
    FalloffRegionScaleFactors=(PositiveWidth=1,NegativeWidth=1,PositiveHeight=1,NegativeHeight=1)
    FalloffSize=0.5
    FlipHeightScaling=true
    FlipWidthScaling=false
    OptionalParentItem=(Type=Bone,Name="spine_01")
    RotationOffset=(X=90,Y=0,Z=-10)
   )
   > n2 = &RigUnit_SphericalPoseReader(
    ActiveRegionScaleFactors=(PositiveWidth=3.5,NegativeWidth=0.5,PositiveHeight=2,NegativeHeight=3)
    ActiveRegionSize=0.1
    Debug=(bDrawDebug=false,DebugScale=25,DebugSegments=20,DebugThickness=0.25)
    DriverAxis=(X=1,Y=0,Z=0)
    DriverItem=(Type=Bone,Name="thigh_r")
    FalloffRegionScaleFactors=(PositiveWidth=1,NegativeWidth=1,PositiveHeight=1,NegativeHeight=1)
    FalloffSize=0.5
    FlipHeightScaling=false
    FlipWidthScaling=true
    OptionalParentItem=(Type=Bone,Name="spine_01")
    RotationOffset=(X=-90,Y=0,Z=-10)
   )
   > &RigUnit_ModifyTransforms(
    ItemToModify="((Item=(Type=Bone,Name=\"dyn_front_flap_01_base\"),Transform=(Rotation=(X=0,Y=0,Z=-0.7113,W=0.7029),Translation=(X=1.9968,Y=0,Z=0),Scale3D=(X=1,Y=1,Z=1))),(Item=(Type=Bone,Name=\"dyn_side_flap_01_l_base\"),Transform=(Rotation=(X=-0.1925,Y=-0.0554,Z=0.0029,W=0. ...(truncated)"
    WeightMinimum=0
    Weight=@n1.OutputParam
   )
   > &RigUnit_ModifyTransforms(
    ItemToModify="((Item=(Type=Bone,Name=\"dyn_front_flap_01_base\"),Transform=(Rotation=(X=0,Y=0,Z=-0.7113,W=0.7029),Translation=(X=1.9968,Y=0,Z=0),Scale3D=(X=1,Y=1,Z=1))),(Item=(Type=Bone,Name=\"dyn_side_flap_01_r_base\"),Transform=(Rotation=(X=-0.2198,Y=-0.1136,Z=0.0019,W=0. ...(truncated)"
    WeightMinimum=0
    Weight=@n2.OutputParam
   )
   Add = &RigVMFunction_MathDoubleAdd(B=@n1.OutputParam, A=@n2.OutputParam)
   Remap = &RigVMFunction_MathDoubleRemap(SourceMaximum=2, SourceMinimum=1, TargetMaximum=1, TargetMinimum=0, bClamp=False, Value=@Add.Result)
   > &RigUnit_ModifyTransforms(
    ItemToModify="((Item=(Type=Bone,Name=\"dyn_front_flap_01_base\"),Transform=(Rotation=(X=0,Y=0,Z=0.7071,W=0.7071),Translation=(X=0,Y=0,Z=0),Scale3D=(X=1,Y=1,Z=1))),(Item=(Type=Bone,Name=\"dyn_rear_flap_01_base\"),Transform=(Rotation=(X=0,Y=0,Z=0,W=1),Translation=(X=0,Y=0,Z=0 ...(truncated)"
    WeightMinimum=0
    Weight=@Remap.Result
   )
   > n3 = &RigUnit_SphericalPoseReader(
    ActiveRegionScaleFactors=(PositiveWidth=2,NegativeWidth=1,PositiveHeight=1,NegativeHeight=1)
    ActiveRegionSize=0.1
    Debug=(bDrawDebug=true,DebugScale=25,DebugSegments=20,DebugThickness=0.25)
    DriverAxis=(X=-1,Y=0,Z=0)
    DriverItem=(Type=Bone,Name="thigh_l")
    FalloffRegionScaleFactors=(PositiveWidth=1,NegativeWidth=1,PositiveHeight=1,NegativeHeight=1)
    FalloffSize=0.3
    FlipHeightScaling=true
    FlipWidthScaling=false
    OptionalParentItem=(Type=Bone,Name="spine_01")
    RotationOffset=(X=180,Y=-10,Z=0)
   )
   > n4 = &RigUnit_SphericalPoseReader(
    ActiveRegionScaleFactors=(PositiveWidth=1,NegativeWidth=2,PositiveHeight=1,NegativeHeight=1)
    ActiveRegionSize=0.1
    Debug=(bDrawDebug=true,DebugScale=25,DebugSegments=20,DebugThickness=0.25)
    DriverAxis=(X=1,Y=0,Z=0)
    DriverItem=(Type=Bone,Name="thigh_r")
    FalloffRegionScaleFactors=(PositiveWidth=1,NegativeWidth=1,PositiveHeight=1,NegativeHeight=1)
    FalloffSize=0.4
    FlipHeightScaling=true
    FlipWidthScaling=false
    OptionalParentItem=(Type=Bone,Name="spine_01")
    RotationOffset=(X=0,Y=-10,Z=0)
   )
   > &RigUnit_ModifyTransforms(
    ItemToModify=((Item=(Type=Bone,Name="dyn_side_flap_01_l_base"),Transform=(Rotation=(X=0,Y=-0.4066,Z=0,W=0.9136),Translation=(X=0,Y=0,Z=0),Scale3D=(X=1,Y=1,Z=1))))
    WeightMinimum=0
    Weight=@n3.OutputParam
   )
   > &RigUnit_ModifyTransforms(
    ItemToModify=((Item=(Type=Bone,Name="dyn_side_flap_01_r_base"),Transform=(Rotation=(X=0,Y=-0.4059,Z=0,W=0.9139),Translation=(X=0,Y=0,Z=0),Scale3D=(X=1,Y=1,Z=1))))
    WeightMinimum=0
    Weight=@n4.OutputParam
   )
   > RigVMFunctionReturnNode()
  ]}
 RigVMFunctionLibrary {nodes [
   Deform_Builtin_Pickaxe = RigVMCollapseNode(ExecuteContext=())
   Deform_HipArmor = RigVMCollapseNode(ExecuteContext=())
   Deform_Shoulderpads = RigVMCollapseNode(ExecuteContext=())
  ]}
 RigVMFunction_Sequence_1_1_ContainedGraph {nodes [
   n1 = RigVMFunctionEntryNode()
   n2 = &RigVMFunction_Sequence(ExecuteContext=@n1.ExecuteContext)
   n3 = &RigVMFunction_Sequence(ExecuteContext=@n2.B)
   n4 = &RigVMFunction_Sequence(ExecuteContext=@n3.B)
   RigVMFunctionReturnNode(A=@n2.A, B=@n3.A, C=@n4.A, D=@n4.B)
  ]}
 RigVMModel {nodes [
   RigUnit_BeginExecution = &RigUnit_BeginExecution()
   FN_DeformRig_FNC = RigVMFunctionReferenceNode(
    DeformRig_ Curve="(deformRig_l_upperarm_up_alpha_72_#1=True,deformRig_r_upperarm_up_alpha_73_#2=True,deformRig_l_upperarm_fr_alpha_74_#3=True,deformRig_r_upperarm_fr_alpha_75_CB51B80C4 ...(truncated)"
    DeformRig_Bool="(upperarm_1_#4=True,elbow_3_#5=True,wrist_7_#6=True,groin_8_#7=True,glute_17_#8=True,knee_9_40C408324CA261 ...(truncated)"
    DeformRig_Debug=(upperarm_debug_18_#4=False,wrist_debug_20_#6=False,lat_debug_26_#9=False,groinGlute_debug_27_#7=False)
    DeformRig_Value="(upperarm_up_18_#10=(Rotation=(X=0,Y=0,Z=0,W=1),Translation=(X=0,Y=3,Z=6),Scale3D=(X=1,Y=1.3,Z=1)),upperarm_fwd_fr_19_#11=(Rotation=(X=0,Y=0,Z=0,W=1),Translation=(X=0,Y=10,Z=0),Scale3D=(X=0.5,Y=1,Z= ...(truncated)"
    ExecuteContext=@RigUnit_BeginExecution.ExecutePin
   )
   > Deform_Shoulderpads = RigVMFunctionReferenceNode()
   > Deform_Builtin_Pickaxe = RigVMFunctionReferenceNode()
   > Deform_HipArmor = RigVMFunctionReferenceNode()
   RigUnit_GetTransform = &RigUnit_GetTransform(Item=(Type=Bone,Name="head"), bInitial=False)
   Interpolate = /Script/RigVM.RigVMFunction_MathQuaternionSlerp(A=(X=0,Y=0,Z=0,W=1), T=0.7, B=@RigUnit_GetTransform.Transform.Rotation) {is=struct RigVMUnitNode, template="Interpolate::Execute(in A,in B,in T,out Result)"}
   Multiply = /Script/RigVM.RigVMFunction_MathQuaternionMul(B=(X=0,Y=0,Z=-0.0346,W=0.9994), A=@Interpolate.Result) {is=struct RigVMUnitNode, template="Multiply::Execute(in A,in B,out Result)"}
   RigUnit_SetRotation = &RigUnit_SetRotation(ExecutePin=(), Item=(Type=Bone,Name="head"), Space=GlobalSpace, Weight=1, bInitial=False, Value=@Interpolate.Result)
  ]}
]
hierarchy [
 root {loc=<0 0 0>, rot=<0 0 0 1>}
 pelvis {parent=root, loc=<0 -0.7071 93.1191>, rot=<0.0127 0.707 -0.0127 -0.707>}
 spine_01 {parent=pelvis, loc=<6.7494 -0.6388 0>, rot=<0 0 0.034 -0.9994>}
 spine_02 {parent=spine_01, loc=<7.2518 -0.035 0>, rot=<0 0 0.0238 -0.9997>}
 spine_03 {parent=spine_02, loc=<6.0104 0 0>, rot=<0 0 0.0149 -0.9999>}
 spine_04 {parent=spine_03, loc=<8.1024 0 0>, rot=<0 0 -0.0132 -0.9999>}
 spine_05 {parent=spine_04, loc=<10.0146 0 0>, rot=<0 0 -0.0015 -1>}
 attach_cape {parent=spine_05, loc=<9.017 -11.714 0>, rot=<0.0616 -0.7044 -0.0616 -0.7044>}
 neck_01 {parent=spine_05, loc=<11.3138 0.1529 0>, rot=<0 0 -0.2003 -0.9797>}
 neck_02 {parent=neck_01, loc=<4.5627 0 0>, rot=<0 0 0.0504 -0.9987>}
 head {parent=neck_02, loc=<5.3914 0 0>, rot=<0 0 0.115 -0.9934>}
 attach_backpack {parent=spine_05, loc=<-0.123 -13.473 0>, rot=<-0.0093 -0.707 0.0093 -0.707>}
 clavicle_l {parent=spine_05, loc=<4.3575 -0.1466 -2.3223>, rot=<0.0465 -0.7289 -0.009 -0.6829>}
 upperarm_l {parent=clavicle_l, loc=<15.5111 0 0>, rot=<0.0226 -0.3234 -0.0088 -0.9459>}
 lowerarm_l {parent=upperarm_l, loc=<26.6514 0 0>, rot=<0.0391 -0.0099 -0.2453 -0.9686>}
 hand_l {parent=lowerarm_l, loc=<22.3491 0 0>, rot=<-0.7081 -0.0507 0.0975 -0.6975>}
 thumb_01_l {parent=hand_l, loc=<1.3118 -0.3191 -2.459>, rot=<0.6546 -0.4042 0.0664 -0.6354>}
 thumb_02_l {parent=thumb_01_l, loc=<5.9336 -0.0448 0.0401>, rot=<-0.0037 0.0213 0.298 -0.9543>}
 thumb_03_l {parent=thumb_02_l, loc=<2.2223 0.3682 -0.0907>, rot=<0.0118 -0.013 0.0066 -0.9998>}
 index_metacarpal_l {parent=hand_l, loc=<1.6535 0.0183 -2.96>, rot=<0.0158 -0.0124 -0.0173 -0.9996>}
 index_01_l {parent=index_metacarpal_l, loc=<8.3008 0.0357 -0.0524>, rot=<-0.0044 0.0485 0.0924 -0.9945>}
 index_02_l {parent=index_01_l, loc=<3.3587 -0.3004 0.0452>, rot=<-0.002 0.0027 0.3225 -0.9466>}
 index_03_l {parent=index_02_l, loc=<2.5578 -0.123 -0.0252>, rot=<-0.002 0.0067 0.2558 -0.9667>}
 middle_metacarpal_l {parent=hand_l, loc=<2.1856 -0.1725 -0.5984>, rot=<0.006 0.0168 -0.0329 -0.9993>}
 middle_01_l {parent=middle_metacarpal_l, loc=<7.3275 -0.2447 0.0129>, rot=<-0.0208 0.0042 0.0858 -0.9961>}
 middle_02_l {parent=middle_01_l, loc=<4.0123 0.1401 0.1175>, rot=<-0.0218 0.0258 0.3512 -0.9357>}
 middle_03_l {parent=middle_02_l, loc=<2.6077 0.0534 0.1373>, rot=<0.0082 0.0188 0.1708 -0.9851>}
 ring_metacarpal_l {parent=hand_l, loc=<2.2074 -0.3625 1.3638>, rot=<-0.0284 0.0722 -0.0458 -0.9959>}
 ring_01_l {parent=ring_metacarpal_l, loc=<7.0295 -0.6197 -0.1446>, rot=<-0.0489 -0.0252 0.1892 -0.9804>}
 ring_02_l {parent=ring_01_l, loc=<3.6174 0.4647 -0.1027>, rot=<0 0 0.2806 -0.9598>}
 ring_03_l {parent=ring_02_l, loc=<2.4152 0.114 0.0199>, rot=<0 0 0.1189 -0.9929>}
 pinky_metacarpal_l {parent=hand_l, loc=<1.7616 -0.9753 3.3814>, rot=<-0.0397 0.0814 -0.069 -0.9935>}
 pinky_01_l {parent=pinky_metacarpal_l, loc=<7.1709 -0.4831 -0.1094>, rot=<-0.0654 -0.0035 0.2242 -0.9723>}
 pinky_02_l {parent=pinky_01_l, loc=<2.8274 0.0893 -0.1043>, rot=<0.0037 0.0011 0.2854 -0.9584>}
 pinky_03_l {parent=pinky_02_l, loc=<2.2046 0.0307 -0.0078>, rot=<-0.0002 0.0098 0.1445 -0.9895>}
 deform_wrist_dn_base_l {parent=hand_l, loc=<0 -2.2342 0>, rot=<0 0 0 -1>}
 deform_wrist_dn_l {parent=deform_wrist_dn_base_l, loc=<0 0 0>, rot=<0 0 0 1>}
 deform_wrist_up_base_l {parent=hand_l, loc=<0 2.2682 0>, rot=<0 0 0 -1>}
 deform_wrist_up_l {parent=deform_wrist_up_base_l, loc=<0 0 0>, rot=<0 0 0 1>}
 deform_pickaxe_hand_root_l {parent=hand_l, loc=<4.6609 1.567 0.134>, rot=<0 0 0 1>}
 deform_pickaxe_blade_a_l {parent=deform_pickaxe_hand_root_l, loc=<11.5997 4.0225 -2.4911>, rot=<0.7696 0.0074 0.0019 -0.6384>}
 deform_pickaxe_blade_b_l {parent=deform_pickaxe_hand_root_l, loc=<11.5124 3.4002 4.3897>, rot=<0.5521 0.0264 0.0226 -0.8331>}
 lowerarm_twist_02_l {parent=lowerarm_l, loc=<9.6199 0 0>, rot=<0 0 0 -1>}
 lowerarm_twist_01_l {parent=lowerarm_l, loc=<17.8549 0 0>, rot=<0 0 0 -1>}
 deform_elbow_in_base_l {parent=lowerarm_l, loc=<0.2112 5.259 0.8274>, rot=<0 0 0 -1>}
 deform_elbow_in_l {parent=deform_elbow_in_base_l, loc=<0 0 0>, rot=<0 0 0 1>}
 deform_elbow_out_base_l {parent=lowerarm_l, loc=<0 -4 0>, rot=<0 0 0 -1>}
 deform_elbow_out_l {parent=deform_elbow_out_base_l, loc=<0 0 0>, rot=<0 0 0 1>}
 upperarm_twist_01_l {parent=upperarm_l, loc=<0.0041 0 0>, rot=<0 0 0 1>}
 deform_upperarm_bk_base_l {parent=upperarm_twist_01_l, loc=<1.0813 -7.2193 -1.9453>, rot=<0 0 0 1>}
 deform_upperarm_bk_l {parent=deform_upperarm_bk_base_l, loc=<0 0 0>, rot=<0 0 0 1>}
 deform_upperarm_fr_base_l {parent=upperarm_twist_01_l, loc=<-0.3691 7.0506 -2.765>, rot=<0 0 0 1>}
 deform_upperarm_fr_l {parent=deform_upperarm_fr_base_l, loc=<0 0 0>, rot=<0 0 0 1>}
 deform_upperarm_up_base_l {parent=upperarm_twist_01_l, loc=<-3.2216 -1.1391 3.3311>, rot=<0 0 0 1>}
 deform_upperarm_up_l {parent=deform_upperarm_up_base_l, loc=<0 0 0>, rot=<0 0 0 1>}
 upperarm_twist_02_l {parent=upperarm_l, loc=<13.7941 0 0>, rot=<0 0 0 1>}
 deform_shoulderpad_l_base {parent=clavicle_l, loc=<13.949 -0.9189 7.2911>, rot=<0.0224 -0.3031 -0.0093 -0.9527>}
 deform_shoulderpad_01_l {parent=deform_shoulderpad_l_base, loc=<0 0 0>, rot=<0 0 0 1>}
 deform_shoulderpad_02_l {parent=deform_shoulderpad_01_l, loc=<2.1353 0 4.8413>, rot=<0 0 0 1>}
 deform_shoulderpad_03_l {parent=deform_shoulderpad_02_l, loc=<4.5792 0 -0.9352>, rot=<0 0 0 1>}
 clavicle_r {parent=spine_05, loc=<4.3575 -0.1466 2.3223>, rot=<0.7289 0.0465 -0.6829 0.009>}
 upperarm_r {parent=clavicle_r, loc=<-15.5111 0 0>, rot=<0.0226 -0.3234 -0.0088 -0.9459>}
 lowerarm_r {parent=upperarm_r, loc=<-26.6514 0 0>, rot=<0.0391 -0.0099 -0.2453 -0.9686>}
 hand_r {parent=lowerarm_r, loc=<-22.3491 0 0>, rot=<-0.7081 -0.0507 0.0975 -0.6975>}
 thumb_01_r {parent=hand_r, loc=<-1.3118 0.3191 2.459>, rot=<0.6546 -0.4042 0.0664 -0.6354>}
 thumb_02_r {parent=thumb_01_r, loc=<-5.9336 0.0448 -0.0401>, rot=<-0.0037 0.0213 0.298 -0.9543>}
 thumb_03_r {parent=thumb_02_r, loc=<-2.2223 -0.3682 0.0907>, rot=<0.0118 -0.013 0.0066 -0.9998>}
 index_metacarpal_r {parent=hand_r, loc=<-1.6535 -0.0183 2.96>, rot=<0.0158 -0.0124 -0.0173 -0.9996>}
 index_01_r {parent=index_metacarpal_r, loc=<-8.3008 -0.0357 0.0524>, rot=<-0.0044 0.0485 0.0924 -0.9945>}
 index_02_r {parent=index_01_r, loc=<-3.3587 0.3004 -0.0452>, rot=<-0.002 0.0027 0.3225 -0.9466>}
 index_03_r {parent=index_02_r, loc=<-2.5578 0.123 0.0252>, rot=<-0.002 0.0067 0.2558 -0.9667>}
 middle_metacarpal_r {parent=hand_r, loc=<-2.1856 0.1725 0.5984>, rot=<0.006 0.0168 -0.0329 -0.9993>}
 middle_01_r {parent=middle_metacarpal_r, loc=<-7.3275 0.2447 -0.0129>, rot=<-0.0208 0.0042 0.0858 -0.9961>}
 middle_02_r {parent=middle_01_r, loc=<-4.0123 -0.1401 -0.1175>, rot=<-0.0218 0.0258 0.3512 -0.9357>}
 middle_03_r {parent=middle_02_r, loc=<-2.6077 -0.0534 -0.1373>, rot=<0.0082 0.0188 0.1708 -0.9851>}
 ring_metacarpal_r {parent=hand_r, loc=<-2.2074 0.3625 -1.3638>, rot=<-0.0284 0.0722 -0.0458 -0.9959>}
 ring_01_r {parent=ring_metacarpal_r, loc=<-7.0295 0.6197 0.1446>, rot=<-0.0489 -0.0252 0.1892 -0.9804>}
 ring_02_r {parent=ring_01_r, loc=<-3.6174 -0.4647 0.1027>, rot=<0 0 0.2806 -0.9598>}
 ring_03_r {parent=ring_02_r, loc=<-2.4152 -0.114 -0.0199>, rot=<0 0 0.1189 -0.9929>}
 pinky_metacarpal_r {parent=hand_r, loc=<-1.7616 0.9753 -3.3814>, rot=<-0.0397 0.0814 -0.069 -0.9935>}
 pinky_01_r {parent=pinky_metacarpal_r, loc=<-7.1709 0.4831 0.1094>, rot=<-0.0654 -0.0035 0.2242 -0.9723>}
 pinky_02_r {parent=pinky_01_r, loc=<-2.8274 -0.0893 0.1043>, rot=<0.0037 0.0011 0.2854 -0.9584>}
 pinky_03_r {parent=pinky_02_r, loc=<-2.2046 -0.0307 0.0078>, rot=<-0.0002 0.0098 0.1445 -0.9895>}
 deform_wrist_dn_base_r {parent=hand_r, loc=<-0.0004 2.2345 -0.0001>, rot=<0 0 0 -1>}
 deform_wrist_dn_r {parent=deform_wrist_dn_base_r, loc=<0 0 0>, rot=<0 0 0 1>}
 deform_wrist_up_base_r {parent=hand_r, loc=<0.0003 -2.2684 0.0001>, rot=<0 0 0 -1>}
 deform_wrist_up_r {parent=deform_wrist_up_base_r, loc=<0 0 0>, rot=<0 0 0 1>}
 deform_pickaxe_hand_root_r {parent=hand_r, loc=<-4.6609 -1.567 -0.134>, rot=<0 0 0 1>}
 deform_pickaxe_blade_a_r {parent=deform_pickaxe_hand_root_r, loc=<-11.5997 -4.0225 2.4911>, rot=<0.7696 0.0074 0.0019 -0.6384>}
 deform_pickaxe_blade_b_r {parent=deform_pickaxe_hand_root_r, loc=<-11.5125 -3.4003 -4.3897>, rot=<0.5521 0.0264 0.0226 -0.8331>}
 lowerarm_twist_02_r {parent=lowerarm_r, loc=<-9.6199 0 0>, rot=<0 0 0 -1>}
 lowerarm_twist_01_r {parent=lowerarm_r, loc=<-17.8549 0 0>, rot=<0 0 0 -1>}
 deform_elbow_in_base_r {parent=lowerarm_r, loc=<-0.2111 -5.2591 -0.8275>, rot=<0 0 0 -1>}
 deform_elbow_in_r {parent=deform_elbow_in_base_r, loc=<-0.0001 0 0.0002>, rot=<0 0 0 1>}
 deform_elbow_out_base_r {parent=lowerarm_r, loc=<0 4 -0.0001>, rot=<0 0 0 -1>}
 deform_elbow_out_r {parent=deform_elbow_out_base_r, loc=<0 0 0>, rot=<0 0 0 1>}
 upperarm_twist_01_r {parent=upperarm_r, loc=<-0.0041 0 0>, rot=<0 0 0 1>}
 deform_upperarm_bk_base_r {parent=upperarm_twist_01_r, loc=<-1.0814 7.2192 1.9455>, rot=<0 0 0 1>}
 deform_upperarm_bk_r {parent=deform_upperarm_bk_base_r, loc=<0 0 0>, rot=<0 0 0 1>}
 deform_upperarm_fr_base_r {parent=upperarm_twist_01_r, loc=<0.3689 -7.0506 2.7652>, rot=<0 0 0 1>}
 deform_upperarm_fr_r {parent=deform_upperarm_fr_base_r, loc=<0 0 0>, rot=<0 0 0 1>}
 deform_upperarm_up_base_r {parent=upperarm_twist_01_r, loc=<3.2214 1.1391 -3.3308>, rot=<0 0 0 1>}
 deform_upperarm_up_r {parent=deform_upperarm_up_base_r, loc=<0 0 0>, rot=<0 0 0 -1>}
 upperarm_twist_02_r {parent=upperarm_r, loc=<-13.7941 0 0>, rot=<0 0 0 1>}
 deform_shoulderpad_r_base {parent=clavicle_r, loc=<-13.9489 0.9189 -7.2915>, rot=<0.0224 -0.3031 -0.0093 -0.9527>}
 deform_shoulderpad_01_r {parent=deform_shoulderpad_r_base, loc=<0 0 0>, rot=<0 0 0 -1>}
 deform_shoulderpad_02_r {parent=deform_shoulderpad_01_r, loc=<-2.1358 0 -4.8407>, rot=<0 0 0 -1>}
 deform_shoulderpad_03_r {parent=deform_shoulderpad_02_r, loc=<-4.5789 0 0.9349>, rot=<0 0 0 -1>}
 deform_pec_aim_base_l {parent=spine_04, loc=<7.8732 12.6388 -2.4123>, rot=<0.0102 -0.4602 0.3979 -0.7936>}
 deform_pec_aim_l {parent=deform_pec_aim_base_l, loc=<0 0 0>, rot=<0 0 0 1>}
 deform_pec_base_l {parent=deform_pec_aim_l, loc=<10.766 0 0>, rot=<0 0 0 1>}
 deform_pec_l {parent=deform_pec_base_l, loc=<0.055 0 0>, rot=<0 0 0 1>}
 deform_pec_aim_base_r {parent=spine_04, loc=<7.8727 12.6388 2.4123>, rot=<-0.4602 -0.0102 0.7936 0.3979>}
 deform_pec_aim_r {parent=deform_pec_aim_base_r, loc=<0 0 0>, rot=<0 0 0 1>}
 deform_pec_base_r {parent=deform_pec_aim_r, loc=<-10.7662 -0.0006 -0.0007>, rot=<0 0 0 1>}
 deform_pec_r {parent=deform_pec_base_r, loc=<0 0 0>, rot=<0 0 0 1>}
 dyn_front_flap_01_base {parent=spine_01, loc=<-2.1524 9.9977 0>, rot=<-0.03 -0.0009 0.0293 -0.9991>}
 dyn_front_flap_01 {parent=dyn_front_flap_01_base, loc=<0 0 0>, rot=<0 0 0 1>}
 dyn_front_flap_02 {parent=dyn_front_flap_01, loc=<-10.3189 0 0>, rot=<0 0 0 1>}
 dyn_side_flap_01_l_base {parent=spine_01, loc=<2.2419 -0.5807 -12.1415>, rot=<0.016 0.1406 -0.0023 -0.9899>}
 dyn_side_flap_01_l {parent=dyn_side_flap_01_l_base, loc=<0 0 0>, rot=<0.0161 0 0 -0.9999>}
 dyn_side_flap_02_l {parent=dyn_side_flap_01_l, loc=<-4.3238 0 0>, rot=<0 0 0 1>}
 dyn_side_flap_03_l {parent=dyn_side_flap_02_l, loc=<-6.912 0 0>, rot=<0 0 0 1>}
 dyn_side_flap_01_r_base {parent=spine_01, loc=<2.2418 -0.5807 12.1415>, rot=<0.1406 -0.016 0.9899 -0.0023>}
 dyn_side_flap_01_r {parent=dyn_side_flap_01_r_base, loc=<0 0 0>, rot=<0 0 0 1>}
 dyn_side_flap_02_r {parent=dyn_side_flap_01_r, loc=<4.3237 0 0.0001>, rot=<0 0 0 1>}
 dyn_side_flap_03_r {parent=dyn_side_flap_02_r, loc=<6.9119 0 0>, rot=<0 0 0 1>}
 dyn_rear_flap_01_base {parent=spine_01, loc=<1.4317 -10.6547 0>, rot=<0.9909 0.1346 0 0>}
 dyn_rear_flap_01 {parent=dyn_rear_flap_01_base, loc=<0 0 0>, rot=<0 0 0 1>}
 dyn_rear_flap_02 {parent=dyn_rear_flap_01, loc=<-6.5852 0 0>, rot=<0 0 0 1>}
 dyn_rear_flap_03 {parent=dyn_rear_flap_02, loc=<-7.3503 0 0>, rot=<0 0 0 1>}
 thigh_l {parent=pelvis, loc=<-2.5693 -0.0109 -7.7501>, rot=<0.0012 0.0572 0.0155 -0.9982>}
 calf_l {parent=thigh_l, loc=<-39.0484 0 0>, rot=<0.0006 -0.0161 -0.0391 -0.9991>}
 foot_l {parent=calf_l, loc=<-42.561 0.1077 0.4292>, rot=<0.0418 -0.0455 0.0343 -0.9975>}
 ball_l {parent=foot_l, loc=<-6.1599 13.8264 0.6806>, rot=<0.0001 -0.0001 -0.7063 -0.7079>}
 calf_twist_02_l {parent=calf_l, loc=<-18.8064 -0.0967 0.0397>, rot=<0.0237 0.0095 -0.0035 -0.9997>}
 calf_twist_01_l {parent=calf_l, loc=<-35.813 -0.0967 0.0397>, rot=<0.0237 0.0095 -0.0035 -0.9997>}
 deform_knee_in_base_l {parent=calf_l, loc=<1.298 -5.2957 -0.0496>, rot=<0 0 0 -1>}
 deform_knee_in_l {parent=deform_knee_in_base_l, loc=<0 0 0>, rot=<0 0 0 1>}
 deform_knee_out_base_l {parent=calf_l, loc=<0 4 0>, rot=<0 0 0 -1>}
 deform_knee_out_l {parent=deform_knee_out_base_l, loc=<0 0 0>, rot=<0 0 0 1>}
 thigh_twist_01_l {parent=thigh_l, loc=<-21.6303 0 0>, rot=<0.0013 0 0 -1>}
 deform_groin_base_l {parent=thigh_l, loc=<-2.1908 8.1142 -1.7442>, rot=<0 0 0 -1>}
 deform_groin_l {parent=deform_groin_base_l, loc=<0 0 0>, rot=<0 0 0 1>}
 thigh_r {parent=pelvis, loc=<-2.5693 -0.0109 7.7501>, rot=<0.0572 -0.0012 0.9982 0.0155>}
 calf_r {parent=thigh_r, loc=<39.0484 0 0>, rot=<0.0006 -0.0161 -0.0391 -0.9991>}
 foot_r {parent=calf_r, loc=<42.561 -0.1077 -0.4292>, rot=<0.0418 -0.0455 0.0343 -0.9975>}
 ball_r {parent=foot_r, loc=<6.1599 -13.8264 -0.6806>, rot=<0.0001 -0.0001 -0.7063 -0.7079>}
 calf_twist_02_r {parent=calf_r, loc=<18.8064 0.0967 -0.0397>, rot=<0.0237 0.0095 -0.0035 -0.9997>}
 calf_twist_01_r {parent=calf_r, loc=<35.813 0.0967 -0.0397>, rot=<0.0237 0.0095 -0.0035 -0.9997>}
 deform_knee_in_base_r {parent=calf_r, loc=<-1.298 5.2957 0.0496>, rot=<0 0 0 -1>}
 deform_knee_in_r {parent=deform_knee_in_base_r, loc=<0 0 0>, rot=<0 0 0 1>}
 deform_knee_out_base_r {parent=calf_r, loc=<0 -4 0>, rot=<0 0 0 -1>}
 deform_knee_out_r {parent=deform_knee_out_base_r, loc=<0 0 0>, rot=<0 0 0 1>}
 thigh_twist_01_r {parent=thigh_r, loc=<21.6303 0 0>, rot=<0.0013 0 0 -1>}
 deform_groin_base_r {parent=thigh_r, loc=<2.1908 -8.1142 1.7442>, rot=<0 0 0 -1>}
 deform_groin_r {parent=deform_groin_base_r, loc=<0 0 0>, rot=<0 0 0 1>}
 deform_glute_base_l {parent=pelvis, loc=<-8.8064 -8.8222 -8.4287>, rot=<0.0275 0.0501 -0.4472 -0.8926>}
 deform_glute_l {parent=deform_glute_base_l, loc=<0 0 0>, rot=<0 0 0 1>}
 deform_glute_base_r {parent=pelvis, loc=<-8.8062 -8.8222 8.4287>, rot=<0.0501 -0.0275 0.8926 -0.4472>}
 deform_glute_r {parent=deform_glute_base_r, loc=<0 0 0>, rot=<0 0 0 1>}
]
variables [
 IsParachuteOpen {type=bool, default=False}
 isSkydiving {type=bool, default=False}
]
