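# Physics asset for the PeacefulPoem medium male player body cosmetic, covering its front, rear and left/right side garment flaps.
# Box bodies form short flap chains off tiny base capsules, with extra edge capsules on lower side segments.
# Pelvis and thigh capsules supply the body shapes around the flaps.
# Bases and thighs join the pelvis with free angular motion, while flap joints have narrow angular limits, tightest on the sides.
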PhysicsAsset /BRCosmetics/Characters/Player/Male/Medium/Bodies/M_MED_PeacefulPoem/Meshes/M_MED_PeacefulPoem_Physics
bodies {
 dyn_front_flap_01 [{shapeType=Box, extentX=12.1875, extentY=4.519, extentZ=22.5825, shapeName=dyn_front_flap_01_box, center=<-5.9934 1.1234 0>}]
 dyn_front_flap_01_base [{shapeType=Capsule, length=0.505, radius=0.505, shapeName=dyn_front_flap_01_base_capsule, rotation=<rot 0 -90 0>}]
 dyn_front_flap_02 [{shapeType=Box, extentX=9.6875, extentY=6.5736, extentZ=22.2197, shapeName=dyn_front_flap_02_box, center=<-6.1145 0.5126 0>}]
 dyn_rear_flap_01 [{
   shapeType=Box
   extentX=24.25
   extentY=4.375
   extentZ=22.6875
   shapeName=dyn_rear_flap_01_box
   center=<-11.6125 -0.6261 0>
   rotation=<rot 0 10.6728 0>
  }]
 dyn_rear_flap_01_base [{shapeType=Capsule, length=0.505, radius=0.505, shapeName=dyn_rear_flap_01_base_capsule, rotation=<rot 0 -90 0>}]
 dyn_side_flap_01_l [{
   shapeType=Box
   extentX=25.8125
   extentY=18.375
   extentZ=3.875
   shapeName=dyn_side_flap_01_l_box
   center=<-15.3399 0 -1.2749>
   rotation=<rot -3.592 -0.0002 0.0001>
  }]
 dyn_side_flap_01_l_base [{shapeType=Capsule, length=0.505, radius=0.505, shapeName=dyn_side_flap_01_l_base_capsule, rotation=<rot 0 -90 0>}]
 dyn_side_flap_01_r [{
   shapeType=Box
   extentX=25.8125
   extentY=18.375
   extentZ=4.9588
   shapeName=dyn_side_flap_01_l_box
   center=<15.3399 0 1.2749>
   rotation=<rot -3.592 -0.0002 -179.9999>
  }]
 dyn_side_flap_01_r_base [{shapeType=Capsule, length=0.505, radius=0.505, shapeName=dyn_side_flap_01_l_base_capsule, rotation=<rot 0 -90 180>}]
 dyn_side_flap_02_l [
  {
   shapeType=Capsule
   length=5.4688
   radius=3.6562
   shapeName=dyn_side_flap_02_r_capsule
   center=<-10.0313 12.6055 1.8654>
   rotation=<rot 85.2097 0.0006 -179.9992>
  }
  {
   shapeType=Capsule
   length=5.625
   radius=3.5625
   shapeName=dyn_side_flap_02_r_capsule
   center=<-10.0828 -10.4693 1.0794>
   rotation=<rot -83.1988 -179.9997 -0.0006>
  }
  {
   shapeType=Box
   extentX=10.0625
   extentY=17.9375
   extentZ=4.3125
   shapeName=dyn_side_flap_02_l_box
   center=<-9.7216 0 -1.6746>
   rotation=<rot -3.5631 -0.0003 0.0001>
  }
 ]
 dyn_side_flap_02_r [
  {
   shapeType=Capsule
   length=5.4688
   radius=3.6562
   shapeName=dyn_side_flap_02_r_capsule
   center=<10.0313 -12.6055 -1.8654>
   rotation=<rot 85.2097 0.0006 0.0008>
  }
  {
   shapeType=Capsule
   length=5.625
   radius=3.5625
   shapeName=dyn_side_flap_02_r_capsule
   center=<10.0828 10.4693 -1.0794>
   rotation=<rot -83.1988 -179.9997 179.9994>
  }
  {
   shapeType=Box
   extentX=10.0625
   extentY=17.9375
   extentZ=4.3125
   shapeName=dyn_side_flap_02_l_box
   center=<9.7216 0 1.6746>
   rotation=<rot -3.5631 -0.0003 -179.9999>
  }
 ]
 dyn_side_flap_03_l [
  {
   shapeType=Capsule
   length=7.9375
   radius=4.0938
   shapeName=dyn_side_flap_03_r_capsule
   center=<-11.8874 11.4753 2.6129>
   rotation=<rot 83.014 0.0002 -179.9997>
  }
  {
   shapeType=Capsule
   length=5.8125
   radius=3.8125
   shapeName=dyn_side_flap_03_r_capsule
   center=<-11.9395 -10.0954 1.2346>
   rotation=<rot -83.2153 -179.9995 -0.0009>
  }
  {
   shapeType=Box
   extentX=13.375
   extentY=17.875
   extentZ=3.8125
   shapeName=dyn_side_flap_03_l_box
   center=<-12.2066 0 -0.4389>
   rotation=<rot -4.2924 -0.0001 0>
  }
 ]
 dyn_side_flap_03_r [
  {
   shapeType=Capsule
   length=7.9375
   radius=4.0938
   shapeName=dyn_side_flap_03_r_capsule
   center=<11.8874 -11.4753 -2.6129>
   rotation=<rot 83.014 0.0002 0.0003>
  }
  {
   shapeType=Capsule
   length=5.8125
   radius=3.8125
   shapeName=dyn_side_flap_03_r_capsule
   center=<11.9395 10.0954 -1.2346>
   rotation=<rot -83.2153 -179.9995 179.9991>
  }
  {
   shapeType=Box
   extentX=13.375
   extentY=17.875
   extentZ=3.8125
   shapeName=dyn_side_flap_03_l_box
   center=<12.2066 0 0.4389>
   rotation=<rot -4.2924 -0.0001 -180>
  }
 ]
 pelvis [{shapeType=Capsule, length=9.3588, radius=10.5539, center=<-1.639 -0.5843 0.0777>, rotation=<rot 0 89.4642 0.1544>}]
 thigh_l [{
   shapeType=Capsule
   length=18.7814
   radius=9.4178
   shapeName=thigh_l_capsule
   center=<-14.1345 -0.5637 0.2959>
   rotation=<rot -16.2219 86.3648 -90.3193>
  }]
 thigh_r [{
   shapeType=Capsule
   length=18.7814
   radius=9.4178
   shapeName=thigh_l_capsule
   center=<14.1345 0.5637 -0.2959>
   rotation=<rot -16.2219 86.3648 89.6807>
  }]
}
constraints [
 {
  bone1Name=thigh_l
  bone2Name=pelvis
  swing1LimitDegrees=45
  swing1Motion=Free
  swing2LimitDegrees=45
  swing2Motion=Free
  twistLimitDegrees=45
  twistMotion=Free
 }
 {bone1Name=dyn_rear_flap_01, bone2Name=dyn_rear_flap_01_base, swing1LimitDegrees=24, swing2LimitDegrees=8, twistLimitDegrees=8}
 {bone1Name=dyn_front_flap_01, bone2Name=dyn_front_flap_01_base, swing1LimitDegrees=16, swing2LimitDegrees=0, twistLimitDegrees=4}
 {bone1Name=dyn_front_flap_02, bone2Name=dyn_front_flap_01, swing1LimitDegrees=16, swing2LimitDegrees=0, twistLimitDegrees=4}
 {
  bone1Name=dyn_front_flap_01_base
  bone2Name=pelvis
  swing1LimitDegrees=45
  swing1Motion=Free
  swing2LimitDegrees=45
  swing2Motion=Free
  twistLimitDegrees=45
  twistMotion=Free
 }
 {
  bone1Name=dyn_side_flap_01_l_base
  bone2Name=pelvis
  swing1LimitDegrees=45
  swing1Motion=Free
  swing2LimitDegrees=45
  swing2Motion=Free
  twistLimitDegrees=45
  twistMotion=Free
 }
 {bone1Name=dyn_side_flap_01_l, bone2Name=dyn_side_flap_01_l_base, swing1LimitDegrees=2, swing2LimitDegrees=5, twistLimitDegrees=2}
 {bone1Name=dyn_side_flap_01_r, bone2Name=dyn_side_flap_01_r_base, swing1LimitDegrees=2, swing2LimitDegrees=5, twistLimitDegrees=2}
 {bone1Name=dyn_side_flap_02_r, bone2Name=dyn_side_flap_01_r, swing1LimitDegrees=2, swing2LimitDegrees=5, twistLimitDegrees=2}
 {bone1Name=dyn_side_flap_03_r, bone2Name=dyn_side_flap_02_r, swing1LimitDegrees=2, swing2LimitDegrees=5, twistLimitDegrees=2}
 {
  bone1Name=dyn_side_flap_01_r_base
  bone2Name=pelvis
  swing1LimitDegrees=45
  swing1Motion=Free
  swing2LimitDegrees=45
  swing2Motion=Free
  twistLimitDegrees=45
  twistMotion=Free
 }
 {bone1Name=dyn_side_flap_02_l, bone2Name=dyn_side_flap_01_l, swing1LimitDegrees=2, swing2LimitDegrees=5, twistLimitDegrees=2}
 {bone1Name=dyn_side_flap_03_l, bone2Name=dyn_side_flap_02_l, swing1LimitDegrees=2, swing2LimitDegrees=5, twistLimitDegrees=2}
 {
  bone1Name=dyn_rear_flap_01_base
  bone2Name=pelvis
  swing1LimitDegrees=45
  swing1Motion=Free
  swing2LimitDegrees=45
  swing2Motion=Free
  twistLimitDegrees=45
  twistMotion=Free
 }
 {
  bone1Name=thigh_r
  bone2Name=pelvis
  swing1LimitDegrees=45
  swing1Motion=Free
  swing2LimitDegrees=45
  swing2Motion=Free
  twistLimitDegrees=45
  twistMotion=Free
 }
]
