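# Animation Blueprint for the face accessory part of the M_MED_PeacefulPoem character body, adding secondary motion to its head plates.
# It copies the attached parent mesh's pose, runs the accessory's Control Rig, then chains spring-damped Anim Dynamics on the upper and lower plate bones.
# The upper plate simulates relative to the head, the lower relative to the upper, with tight angular limits and reduced gravity.
# On customization load it caches the Head part mesh, which the AnimGraph does not read.
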AnimBlueprint /BRCosmetics/Characters/Player/Male/Medium/Bodies/M_MED_PeacefulPoem/Meshes/Parts/M_MED_PeacefulPoem_FaceAcc_AnimBP
target_skeleton=&/M_MED_PeacefulPoem_FaceAcc_Skeleton
graphs [
 AnimGraph {nodes [
   n1 = Control Rig(Source=Copy Pose From Mesh() {bUseAttachedParent=true}) {
    alphaBoolBlend {blendOption=Linear}
    controlRigAssetReference {blueprintRigClass=&/M_MED_PeacefulPoem_FaceAcc_CtrlRig.M_MED_PeacefulPoem_FaceAcc_CtrlRig_C}
   }
   n2 = Anim Dynamics - Bone: dyn_head_plate_upper(ComponentPose=Local To Component(LocalPose=@n1), Alpha=0.85, ExternalForce="0,0,0") {
    angularDampingOverride=0.98
    angularSpringConstant=10
    bAngularSpring=true
    bOverrideAngularDamping=true
    bUseGravityOverride=true
    gravityScale=0.35
    simulationSpace=BoneRelative
    alphaBoolBlend {blendOption=Linear}
    boundBone {boneName=dyn_head_plate_upper}
    componentAppliedLinearAccClamp=<1400 1400 1400>
    componentLinearAccScale=<1 1 1>
    componentLinearVelScale=<1 1 1>
    gravityOverride {z=-980}
    physicsBodyDefinitions [{
      boundBone {boneName=dyn_head_plate_upper}
      boxExtents=<30 30 30>
      constraintSetup {angularLimitsMax=<3 3 3>, angularLimitsMin=<-3 -3 -3>, angularTarget=<-1 0 0>}
      localJointOffset=<-5 0 0>
     }]
    relativeSpaceBone {boneName=head}
   }
   n3 = Anim Dynamics - Bone: dyn_head_plate_lower(ComponentPose=@n2, Alpha=0.85, ExternalForce="0,0,0") {
    angularDampingOverride=0.98
    angularSpringConstant=10
    bAngularSpring=true
    bOverrideAngularDamping=true
    bUseGravityOverride=true
    gravityScale=0.35
    simulationSpace=BoneRelative
    alphaBoolBlend {blendOption=Linear}
    boundBone {boneName=dyn_head_plate_lower}
    componentAppliedLinearAccClamp=<1400 1400 1400>
    componentLinearAccScale=<1 1 1>
    componentLinearVelScale=<1 1 1>
    gravityOverride {z=-980}
    physicsBodyDefinitions [{
      boundBone {boneName=dyn_head_plate_lower}
      boxExtents=<30 30 30>
      constraintSetup {angularLimitsMax=<3 3 3>, angularLimitsMin=<-3 -3 -3>, angularTarget=<-1 0 0>}
      localJointOffset=<-5 0 0>
     }]
    relativeSpaceBone {boneName=dyn_head_plate_upper}
   }
   Output Pose(Result=Component To Local(ComponentPose=@n3)) {layerGroup=None}
  ]}
 EventGraph {nodes [
   n1 = Event OnCharacterCustomizationLoaded()
   > Set HeadSkelMesh(HeadSkelMesh:Skeletal Mesh Component Budgeted Object Reference=Get Array Item with EFort Custom Part Type(Array=@n1.CustomMeshParts, PartType=Head)) -> (Output_Get:Skeletal Mesh Component Budgeted Object Reference)
   Event BlueprintUpdateAnimation()
  ]}
]
variables [HeadSkelMesh {type=SkeletalMeshComponentBudgeted}]
